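# Conversation database for the BalletAssassin NPC, defining its greeting and top-level conversation branches.
# The entry point leads to one Speech node whose contextual lines cover first-ever meetings, reactions to specific player outfits, and lines gated on per-match conversation count.
# Priority curves rank the first-meeting and outfit lines, with a default message as fallback.
# The speech then links out by tag to NPC Library quests, basic services and the NPC's own Dialogue entry.
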ConversationDatabase /NPCLibrary/NPCs/BalletAssassin/NPCLibrary_Conversation_BalletAssassin
entry_points [{tag=Conversation.Entry.BalletAssassin, destinations [#1]}]
nodes [
 ConversationEntryPointNode_7 {
  id=#1
  class=ConversationEntryPointNode
  outputs [#2]
  properties {entryTag {tagName=Conversation.Entry.BalletAssassin}}
 }
 Speech_C_4 {
  id=#2
  class=Speech_C
  outputs [#3, #4, #5]
  properties {
   nodeName=""
   generalConfig {
    defaultMessage=You lack grace. But your form is adequate.
    contextualMessages [
     {
      message=You are lucky to meet me.
      contextRequirements [{requirement {class=FortControllerRequirement_IsFirstEverConversationWithNPC}}]
      priority {curve {rowName=Tandem.Speech.Priority.FirstEverInteraction}}
     }
     {
      message="Another assignment for me, then? Make it quick."
      contextRequirements [{requirement {class=FortControllerRequirement_HasCID, properties {allowedCIds [/BRCosmetics/Athena/Items/Cosmetics/Characters/Character_MadameMoth_Posh]}}}]
      priority {curve {rowName=Tandem.Speech.Priority.OutfitReaction}}
     }
     {
      message="You were always too naive for this vicious world, little Hope."
      contextRequirements [{requirement {
         class=FortControllerRequirement_HasCID
         properties {allowedCIds [&Character_VoidRedemption_Rebel]}
        }}]
      priority {curve {rowName=Tandem.Speech.Priority.OutfitReaction}}
     }
     {
      message="You are beneath me now, thief. Perhaps you always were."
      contextRequirements [{requirement {
         class=FortControllerRequirement_HasCID
         properties {allowedCIds [/BRCosmetics/Athena/Items/Cosmetics/Characters/Character_DiamondHeart_Chic]}
        }}]
      priority {curve {rowName=Tandem.Speech.Priority.OutfitReaction}}
     }
     {
      message="Without The Society behind you, you are useless to me."
      contextRequirements [{requirement {
         class=FortControllerRequirement_HasCID
         properties {allowedCIds [
           /BRCosmetics/Athena/Items/Cosmetics/Characters/Character_CeremonialGuard_Fencer
           &Character_BengalBasher_Suit
           /BRCosmetics/Athena/Items/Cosmetics/Characters/Character_NovelGuard
          ]}
        }}]
      priority {curve {rowName=Tandem.Speech.Priority.OutfitReaction}}
     }
     {
      message=You lack grace. But your form is adequate.
      contextRequirements [{requirement {
         class=FortControllerRequirement_CheckNPCMatchConversationCount
         properties {nPCUniqueIdTag {tagName=AISpawnerData.Type.Tandem.BalletAssassin}}
        }}]
      priority {curve {rowName=None}}
     }
     {
      message=Do not bore me.
      contextRequirements [{requirement {
         class=FortControllerRequirement_CheckNPCMatchConversationCount
         properties {nPCUniqueIdTag {tagName=AISpawnerData.Type.Tandem.BalletAssassin}}
        }}]
      priority {curve {rowName=None}}
     }
     {
      message="Strike so quietly, even a god could not hear you."
      contextRequirements [{requirement {
         class=FortControllerRequirement_CheckNPCMatchConversationCount
         properties {nPCUniqueIdTag {tagName=AISpawnerData.Type.Tandem.BalletAssassin}}
        }}]
      priority {curve {rowName=None}}
     }
     {
      message=My respect is not given. It is earned.
      contextRequirements [{requirement {
         class=FortControllerRequirement_CheckNPCMatchConversationCount
         properties {nPCUniqueIdTag {tagName=AISpawnerData.Type.Tandem.BalletAssassin}}
        }}]
      priority {curve {rowName=None}}
     }
     {
      message=The end comes to everyone. But it comes first to the foolish.
      contextRequirements [{requirement {
         class=FortControllerRequirement_CheckNPCMatchConversationCount
         properties {nPCUniqueIdTag {tagName=AISpawnerData.Type.Tandem.BalletAssassin}}
        }}]
      priority {curve {rowName=None}}
     }
    ]
   }
  }
 }
 ConversationLinkNode_2 {
  id=#5
  class=ConversationLinkNode
  properties {remoteEntryTag {tagName=Conversation.Entry.BalletAssassin.Dialogue}}
 }
 ConversationLinkNode_0 {id=#4, class=ConversationLinkNode, properties {remoteEntryTag {tagName=Conversation.Entry.BasicServices}}}
 ConversationLinkNode_1 {id=#3, class=ConversationLinkNode, properties {remoteEntryTag {tagName=Conversation.Entry.NPCLibrary.Quests}}}
]
