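# Conversation database for the Ballet Assassin NPC's dialogue, letting the player ask about Valeria, The Society and Hope.
# A single tagged entry point opens an "Ask about..." speech that branches into those three topic choices plus a Back option.
# Each topic plays a linear chain of speech lines, and all three converge on one shared ReturnToStart node labeled BACK.
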ConversationDatabase /NPCLibrary/NPCs/BalletAssassin/NPCLibrary_Conversation_BalletAssassin_Dialogue
entry_points [{tag=Conversation.Entry.BalletAssassin.Dialogue, destinations [#1]}]
nodes [
 ConversationEntryPointNode_0 {
  id=#1
  class=ConversationEntryPointNode
  outputs [#2]
  properties {entryTag {tagName=Conversation.Entry.BalletAssassin.Dialogue}}
 }
 Speech_C_0 {
  id=#2
  class=Speech_C
  outputs [#3, #4, #5, #6]
  properties {
   nodeName=""
   softTaskIcon=&T_UI_NPC_question
   generalConfig {defaultMessage="From my dossiers to your ears, it seems."}
  }
  sub_nodes [ConversationChoiceNode_5 {class=ConversationChoiceNode, properties {defaultChoiceDisplayText=Ask about...}}]
 }
 Speech_C_2 {
  id=#6
  class=Speech_C
  outputs [#7]
  properties {
   nodeName=""
   softTaskIcon=&T_UI_NPC_question
   generalConfig {defaultMessage=I've known Hope as long as I've known her sister.}
  }
  sub_nodes [ConversationChoiceNode_10 {class=ConversationChoiceNode, properties {defaultChoiceDisplayText=Hope}}]
 }
 Speech_C_3 {
  id=#7
  class=Speech_C
  outputs [#8]
  properties {nodeName="", softTaskIcon=&T_UI_NPC_question, generalConfig {defaultMessage="That is, a very, very long time."}}
 }
 Speech_C_4 {
  id=#8
  class=Speech_C
  outputs [#9]
  properties {
   nodeName=""
   softTaskIcon=&T_UI_NPC_question
   generalConfig {defaultMessage="She still thinks someone can save the world, poor thing."}
  }
 }
 ReturnToStart_C_23 {
  id=#9
  class=ReturnToStart_C
  properties {softTaskIcon=&T_UI_NPC_Back}
  sub_nodes [ConversationChoiceNode_32 {class=ConversationChoiceNode, properties {defaultChoiceDisplayText=BACK}}]
 }
 Back_C_0 {
  id=#5
  class=Back_C
  sub_nodes [ConversationChoiceNode_7 {class=ConversationChoiceNode, properties {defaultChoiceDisplayText=BACK}}]
 }
 Speech_C_6 {
  id=#4
  class=Speech_C
  outputs [#10]
  properties {
   nodeName=""
   softTaskIcon=&T_UI_NPC_question
   generalConfig {defaultMessage=The Society gave this island much-needed stability.}
  }
  sub_nodes [ConversationChoiceNode_0 {class=ConversationChoiceNode, properties {defaultChoiceDisplayText=The Society}}]
 }
 Speech_C_12 {
  id=#10
  class=Speech_C
  outputs [#11]
  properties {
   nodeName=""
   softTaskIcon=&T_UI_NPC_question
   generalConfig {defaultMessage=Though offering a spot to riffraff like Oscar... Hmph.}
  }
 }
 Speech_C_13 {
  id=#11
  class=Speech_C
  outputs [#12]
  properties {
   nodeName=""
   softTaskIcon=&T_UI_NPC_question
   generalConfig {defaultMessage="This Island will always crave a strong hand. "}
  }
 }
 Speech_C_1 {
  id=#12
  class=Speech_C
  outputs [#9]
  properties {
   nodeName=""
   softTaskIcon=&T_UI_NPC_question
   generalConfig {defaultMessage=Do you expect to defeat these gods and rule yourself?}
  }
 }
 Speech_C_7 {
  id=#3
  class=Speech_C
  outputs [#13]
  properties {
   nodeName=""
   softTaskIcon=&T_UI_NPC_question
   generalConfig {defaultMessage=Do you really think I'll inform on my top client?}
  }
  sub_nodes [ConversationChoiceNode_15 {class=ConversationChoiceNode, properties {defaultChoiceDisplayText=Valeria}}]
 }
 Speech_C_15 {
  id=#13
  class=Speech_C
  outputs [#14]
  properties {
   nodeName=""
   softTaskIcon=&T_UI_NPC_question
   generalConfig {defaultMessage=She's the only one of those Society brats whose calls I'll still take.}
  }
 }
 Speech_C_14 {
  id=#14
  class=Speech_C
  outputs [#9]
  properties {
   nodeName=""
   softTaskIcon=&T_UI_NPC_question
   generalConfig {defaultMessage="Am I on a job for her now? You'd love to know, wouldn't you."}
  }
 }
]
